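# Injected AI behavior tree for the secondary chain whip attack, which has a bot use its injected item in alternate mode.
# A timed Wait is the main task of a Simple Parallel, bounding a nested parallel that holds the alternate input while a delayed sequence sends a primary-pressed gameplay event as the hold time elapses.
# A root service keeps focus on the target, and the root Selector falls back to a short Wait if the attack branch fails.
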BehaviorTree /SunRoseChainWhipGameplay/InjectedBehaviors/ChainsAttackBT_Secondary
blackboard=&/ChainsAttackBB
root {
 name=Selector
 class=BTComposite_Selector
 children [
  Simple Parallel {
   class=BTComposite_SimpleParallel
   children [
    Wait {class=BTTask_Wait, properties {waitTime {defaultValue=3}}}
    Simple Parallel {
     class=BTComposite_SimpleParallel
     children [
      Use Item {
       class=FortAthenaBTTask_UseItem
       decorators []
       properties {
        actionObjectKeyName=AIEvaluator_Injection_Item
        executionStatusKeyName=None
        nodeName=Use Item
        finishAfterItemUse {defaultValue=false}
        inputHoldDuration {defaultValue=1.5, key=None}
        maximumWaitTimeBetweenUses {defaultValue=1}
        resetActionObjectKey {defaultValue=false}
        useAlternateMode {defaultValue=true}
        validateAbility {defaultValue=false}
        validateAbilityBeforeEquipping {defaultValue=false}
       }
      }
      Sequence {
       class=BTComposite_Sequence
       children [
        Wait {class=BTTask_Wait, properties {waitTime {defaultValue=1.5}}}
        BTT_SendGameplayEvent_C_0 {
         class=BTT_SendGameplayEvent_C
         children []
         decorators []
         properties {
          bIgnoreRestartSelf=false
          bShowPropertyDetails=true
          customDescription=""
          nodeName=""
          taskSucceeded=false
          gameplayEventTag {key=None, defaultValue {gameplayTags [{tagName=Event.Primary.Pressed}]}}
          tickInterval {interval=-1}
         }
         services []
        }
       ]
       decorators []
       properties {children [
         {childTask=&self:BTTask_Wait_2}
         {childTask=&self:BTT_SendGameplayEvent_C_0}
        ]}
      }
     ]
     properties {
      children [
       {childTask=&self:FortAthenaBTTask_UseItem_0}
       {childComposite=&self:BTComposite_Sequence_2, childTask=null}
      ]
      services []
     }
     services []
    }
   ]
   properties {
    children [
     {childTask=&self:BTTask_Wait_1}
     {childComposite=&self:BTComposite_SimpleParallel_1, childTask=null}
    ]
    services []
   }
   services []
  }
  Wait {class=BTTask_Wait, properties {waitTime {defaultValue=2}}}
 ]
 decorators []
 properties {
  children [
   {
    childComposite=&self:BTComposite_SimpleParallel_0
    childTask=null
    decorators []
   }
   {childComposite=null, childTask=&self:BTTask_Wait_0, decorators []}
  ]
  services [&self:FortAthenaBTService_AIEvaluator_55]
 }
 services [Focus on target {
   class=FortAthenaBTService_AIEvaluator
   properties {interval=0.5, nodeName=Focus on target, aIEvaluatorClass {defaultValue=&FortAthenaAIBotEvaluator_HandleFocusing}}
  }]
}
